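# Lobby UI widget that displays one Habanero tier category as a titled column of icon tiles with a side divider.
# UpdateList sets the title, then reverse-loops the tier structs, creating icon-with-text tiles colored by tier type data and labeled only when several tiers exist.
# The divider hides for the HabaneroSauce8 category.
# A second loop applies index-based icon scaling, but references only the last created tile.
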WidgetBlueprint /Game/UI/Frontend/Lobby/Habanero/WBP_Habanero_CategoryDisplay
graphs [
 EventGraph {nodes [
   Event PreConstruct()
   Event Construct()
   Event Tick()
  ]}
 UpdateList {nodes [
   n1 = UpdateList!() -> (then, Name:Text, List:Array of Fort Habanero Tier Structures) {is=title K2Node_FunctionEntry}
   > "SetText (Text)"(self=Get Label() -> (Label:Common Text Object Reference), InText=@n1)
   > n2 = Reverse for Each Loop!(Exec:Exec, Array:Array of Fort Habanero Tier Structures=Get List() -> (List:Array of Fort Habanero Tier Structures)) -> (LoopBody:Exec, ArrayIndex:int, ArrayElement:Fort Habanero Tier Structure, Completed:Exec) {is=title K2Node_MacroInstance}
   n3 = For Each Loop(Exec=@n2.Completed, Array:Array of Fort Habanero Tier Structures=Get List()) -> (Array Element:Fort Habanero Tier Structure)
   n4 = Create Widget!(
    execute:Exec=@n2.LoopBody
    Class:User Widget Class Reference=&/WBP_Habanero_IconWithText.WBP_Habanero_IconWithText_C
    OwningPlayer:Player Controller Object Reference
   ) -> (then:Exec, ReturnValue:WBP Habanero Icon With Text Object Reference)
   SetSize!(
    execute=@n3.LoopBody
    self:WBP Habanero Icon Object Reference=Get WBP_Habanero_Icon!(self:WBP Habanero Icon With Text Object Reference=@n4) -> (WBP_Habanero_Icon:WBP Habanero Icon Object Reference)
    IconSize:Vector 2D Structure=vector2d * vector2d(
     A:Vector 2D Structure=Select!(
      Option 0:Vector 2D Structure=(X=1,Y=1)
      Option 1:Vector 2D Structure=(X=0.775,Y=0.775)
      Option 2:Vector 2D Structure=(X=0.7,Y=0.7)
      Index:int=@n3.Array Index
     ) -> (ReturnValue:Vector 2D Structure)
     B:Vector 2D Structure=(X=96,Y=96)
    ) -> (ReturnValue:Vector 2D Structure) {is=title K2Node_PromotableOperator}
   ) -> (then)
   n5 = Break Fort Habanero Tier!(FortHabaneroTier:"Fort Habanero Tier Structure (by ref)"=@n2.ArrayElement) -> (
    ForegroundIcon:Material Instance Soft Object Reference
    ForegroundIconSmall:Material Instance Soft Object Reference
    BackgroundIcon:Material Instance Soft Object Reference
    BackgroundIconSmall:Material Instance Soft Object Reference
    WhiteIcon:Texture 2D Soft Object Reference
    SimpleIconSmall:Texture 2D Soft Object Reference
    Name:Text
    TierType:Fort Habanero Tier Type Data Object Reference
    AdditionalLeftIcon:Material Instance Soft Object Reference
    AdditionalRightIcon:Material Instance Soft Object Reference
   ) {is=title K2Node_BreakStruct}
   n6 = Get Category!(self:Fort Habanero Tier Type Data Object Reference=@n5.TierType) -> (Category:EHabaneroTierType Enum=HabaneroSauce0)
   SetVisibility(
    execute=@n10.then
    self=Get Divider() -> (Divider:Image Object Reference)
    InVisibility=Select!(
     Option 0:ESlateVisibility Enum=SelfHitTestInvisible
     Option 1:ESlateVisibility Enum=Hidden
     Index:bool="Equal (Enum)"(A:EHabaneroTierType Enum=@n6, B:EHabaneroTierType Enum=HabaneroSauce8)
    ) -> (ReturnValue:ESlateVisibility Enum)
   )
   n7 = Length(TargetArray:Array of Fort Habanero Tier Structures=Get List())
   AddChild(execute=@n4.then, self=Get Vbox_Tiles() -> (Vbox_Tiles:Vertical Box Object Reference), Content=@n4)
   > n8 = Cast To BP_UI_Habanero_TierTypeDataBase!(execute, Object=@n5.TierType) -> (then, CastFailed, AsBP UI Habanero Tier Type Data Base:BP UI Habanero Tier Type Data Base Object Reference) {is=title K2Node_DynamicCast}
   n9 = Get Tier() -> (Tier:Array of Texts)
   > n10 = SetupTile!(
    execute
    self:WBP Habanero Icon With Text Object Reference=@n4
    IconForeground:Material Instance Soft Object Reference=@n5.ForegroundIconSmall
    IconBackground:Material Instance Soft Object Reference=@n5.BackgroundIconSmall
    Label:Text=Select(
     Option 0:Text
     Option 1:Text="Get (a copy)"(Array:Array of Texts=@n9, Dimension 1=@n2) -> (Output:Text)
     Index="integer > integer"(A=@n7, B=1)
    ) -> (ReturnValue:Text)
    Label Color:Color Structure=Get Text_Color!(self:BP UI Habanero Tier Type Data Base Object Reference=@n8) -> (Text_Color:Color Structure)
   ) -> (then)
  ]}
]
variables []
widgets {
 info {inheritedWidgetCount=0, parentClass=&UserWidget, rootWidgetClass=&HorizontalBox, widgetCount=5}
 widgets [
  HorizontalBox_0 {widgetClassPath=&HorizontalBox, widget=HorizontalBox_0}
  VerticalBox_25 {
   widgetClassPath=&VerticalBox
   parent=HorizontalBox_0
   widget=VerticalBox_25
   slot {class=HorizontalBoxSlot, properties {horizontalAlignment=HAlign_Center, padding=<ltrb 0 32 0 0>, size {sizeRule=Fill, value=1}}}
  }
  Label {
   widgetClassPath=&CommonTextBlock
   bIsVariable=true
   parent=VerticalBox_25
   widget=Label
   properties {
    justification=Center
    mobileFontSizeMultiplier=1.2
    scrollStyle=/FortUILibrary/Styles/Typography/Foundation/TextScrollStyle-NoFade.TextScrollStyle-NoFade_C
    text=Label
    textTransformPolicy=ToUpper
    font {
     bForceMonospaced=false
     bMaterialIsStencil=false
     fontMaterial=null
     fontObject=&BurbankSmall
     letterSpacing=0
     monospacedWidth=1
     size=16
     skewAmount=0
     typefaceFontName=Black
     outlineSettings {
      bApplyOutlineToDropShadows=false
      bMiteredCorners=false
      bSeparateFillAlpha=false
      outlineMaterial=null
      outlineSize=0
      outlineColor=<rgba 0 0 0 1>
     }
    }
    shadowOffset=<0 0>
   }
   slot {class=VerticalBoxSlot, properties {padding=<ltrb 6 0 6 16>}}
  }
  Vbox_Tiles {widgetClassPath=&VerticalBox, bIsVariable=true, parent=VerticalBox_25, widget=Vbox_Tiles, slot {class=VerticalBoxSlot}}
  Divider {
   widgetClassPath=&Image
   bIsVariable=true
   parent=HorizontalBox_0
   widget=Divider
   properties {
    renderOpacity=0.5
    visibility=SelfHitTestInvisible
    brush {
     bIsDynamicallyLoaded=false
     drawAs=Image
     imageType=NoImage
     mirroring=NoMirror
     resourceName=None
     resourceObject=/Game/UI/Frontend/Lobby/Habanero/Habanero_Materials/MaterialInstances/Background/MI_UI_Habanero_LinearGradientTop
     tiling=NoTile
     imageSize=<2 2>
     margin=<ltrb 0 0 0 0>
     outlineSettings {
      bUseBrushTransparency=false
      roundingType=HalfHeightRadius
      width=0
      color {colorUseRule=UseColor_Specified, specifiedColor=<rgba 0 0 0 0>}
      cornerRadii=<0 0 0 0>
     }
     tintColor {colorUseRule=UseColor_Specified, specifiedColor=<rgba 1 1 1 1>}
     uVRegion {bIsValid=false, max=<0 0>, min=<0 0>}
    }
    colorAndOpacity=<rgba 0.0009 0.006 0.0395 1>
   }
   slot {class=HorizontalBoxSlot}
  }
 ]
}
