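# Intro entry widget for the lobby's Habanero info list, showing rank tier categories above a titled bullet list.
# Setting the game mode rebuilds the row from RankData, batching consecutive same-category tiers into equal-fill CategoryDisplay widgets, flushed on category change and loop completion.
# An InfoPanel data object supplies the title and bullets through a DynamicEntryBox.
# Focus animations fade content and tint the border, while spacers pad first and last entries.
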WidgetBlueprint /Game/UI/Frontend/Lobby/Habanero/HabaneroInfoEntry/WBP_Habanero_InfoIntroEntry
animations [
 Unfocused {
  binding_count=3
  length_seconds=0.25
  object=&self:Unfocused
  sequence {bindings [
    Categories {
     id=#1
     possessed_class=HorizontalBox
     tracks [{
       type=MovieSceneFloatTrack
       display_name=Render Opacity
       sections [{type=MovieSceneFloatSection, end_frame=1, start_frame=0, channels [None {type=MovieSceneScriptingFloatChannel, keys [[0,1],[1,0]]}]}]
      }]
    }
    Border_0 {
     id=#2
     possessed_class=Border
     tracks [{
       type=MovieSceneColorTrack
       display_name=Content Color and Opacity
       sections [{
         type=MovieSceneColorSection
         end_frame=5
         start_frame=0
         channels [
          Color.R {type=MovieSceneScriptingFloatChannel, keys [[0,1],[5,0.1302]]}
          Color.G {type=MovieSceneScriptingFloatChannel, keys [[0,1],[5,0.1302]]}
          Color.B {type=MovieSceneScriptingFloatChannel, keys [[0,1],[5,0.1302]]}
          Color.A {type=MovieSceneScriptingFloatChannel, const [1,1]}
         ]
        }]
      }]
    }
    VerticalBox_Text {
     id=#3
     possessed_class=VerticalBox
     tracks [{
       type=MovieSceneFloatTrack
       display_name=Render Opacity
       sections [{type=MovieSceneFloatSection, end_frame=5, start_frame=0, channels [None {type=MovieSceneScriptingFloatChannel, keys [[0,1],[1,0]]}]}]
      }]
    }
   ]}
 }
 Focused {
  binding_count=3
  length_seconds=0.2521
  object=&self:Focused
  sequence {bindings [
    Border_0 {
     id=#2
     possessed_class=Border
     tracks [{
       type=MovieSceneColorTrack
       display_name=Content Color and Opacity
       sections [{
         type=MovieSceneColorSection
         end_frame=5
         start_frame=0
         channels [
          Color.R {type=MovieSceneScriptingFloatChannel, keys [[0,0.1302],[5.0427,1]]}
          Color.G {type=MovieSceneScriptingFloatChannel, keys [[0,0.1302],[5.0427,1]]}
          Color.B {type=MovieSceneScriptingFloatChannel, keys [[0,0.1302],[5.0427,1]]}
          Color.A {type=MovieSceneScriptingFloatChannel, const [1,2]}
         ]
        }]
      }]
    }
    Categories {
     id=#4
     possessed_class=HorizontalBox
     tracks [{
       type=MovieSceneFloatTrack
       display_name=Render Opacity
       sections [{
         type=MovieSceneFloatSection
         end_frame=5
         start_frame=0
         channels [None {type=MovieSceneScriptingFloatChannel, keys [[0,0],[1.0087,0],[4.0343,1],[5.043,1]]}]
        }]
      }]
    }
    VerticalBox_Text {
     id=#5
     possessed_class=VerticalBox
     tracks [{
       type=MovieSceneFloatTrack
       display_name=Render Opacity
       sections [{type=MovieSceneFloatSection, end_frame=5, start_frame=0, channels [None {type=MovieSceneScriptingFloatChannel, keys [[0,0],[5.043,1]]}]}]
      }]
    }
   ]}
 }
]
graphs [
 EventGraph {nodes [
   n1 = Event OnSetGameMode!() -> (OutputDelegate, then, GameMode:String) {is=title K2Node_Event}
   > Update Grid!(execute, self, GameMode:String=@n1.GameMode) -> (then)
  ]}
 Update Grid {nodes [
   n1 = Update Grid!() -> (then, GameMode:String) {is=title K2Node_FunctionEntry}
   > ClearChildren(self=Get Categories() -> (Categories:Horizontal Box Object Reference))
   n2 = Get EntryData!(self:Fort Habanero Display Data Object Reference=Get RankData() -> (RankData:Fort Habanero Display Data Object Reference)) -> (EntryData:Map of Strings to Fort Habanero Mode List Structures)
   > n3 = For Each Loop(Array:Array of Fort Habanero Tier Structures=Break Fort Habanero Mode List!(FortHabaneroModeList:"Fort Habanero Mode List Structure (by ref)"=Find(TargetMap:Map of Strings to Fort Habanero Mode List Structures=@n2, Key=@n1) -> (Value:Fort Habanero Mode List Structure)) -> (TierList:Array of Fort Habanero Tier Structures, bShowTopLeaderboardIcon:bool=~, TopTierTopLeaderboardPositions:Fort Habanero Tier Structure) {is=title K2Node_BreakStruct}) -> (Array Element:Fort Habanero Tier Structure)
   > n4 = Cast To BP_UI_Habanero_TierTypeDataBase!(
    execute
    Object=Break Fort Habanero Tier!(FortHabaneroTier:"Fort Habanero Tier Structure (by ref)"=@n3) -> (
     ForegroundIcon:Material Instance Soft Object Reference
     ForegroundIconSmall:Material Instance Soft Object Reference
     BackgroundIcon:Material Instance Soft Object Reference
     BackgroundIconSmall:Material Instance Soft Object Reference
     WhiteIcon:Texture 2D Soft Object Reference
     SimpleIconSmall:Texture 2D Soft Object Reference
     Name:Text
     TierType:Fort Habanero Tier Type Data Object Reference
     AdditionalLeftIcon:Material Instance Soft Object Reference
     AdditionalRightIcon:Material Instance Soft Object Reference
    ) {is=title K2Node_BreakStruct}.TierType
   ) -> (then, CastFailed, AsBP UI Habanero Tier Type Data Base:BP UI Habanero Tier Type Data Base Object Reference) {is=title K2Node_DynamicCast}
   > n5 = Sequence()
   > Set CurrentCategory(CurrentCategory:EHabaneroTierType Enum=Get Category!(self:Fort Habanero Tier Type Data Object Reference=@n4) -> (Category:EHabaneroTierType Enum=HabaneroSauce0)) -> (Output_Get:EHabaneroTierType Enum=HabaneroSauce0)
   n6 = Get LastCategory() -> (LastCategory:EHabaneroTierType Enum=HabaneroSauce0)
   n7 = Get CurrentCategory() -> (CurrentCategory:EHabaneroTierType Enum=HabaneroSauce0)
   n8 = "Equal (Integer)"(A=@n3.Array Index)
   > Branch(Condition=OR Boolean(A="Not Equal (Enum)"!(A:EHabaneroTierType Enum=@n7, B:EHabaneroTierType Enum=@n6) -> (ReturnValue:bool), B=@n8))
   > n9 = Branch(Condition=@n8)
   Clear(
    execute=[@n12.then, @n9.then]
    TargetArray:Array of Fort Habanero Tier Structures=Get CategoryList() -> (CategoryList:Array of Fort Habanero Tier Structures)
   )
   > Set LastCategory(LastCategory:EHabaneroTierType Enum=Get CurrentCategory() -> (CurrentCategory=HabaneroSauce0)) -> (Output_Get:EHabaneroTierType Enum=HabaneroSauce0)
   > Set LastCategoryName(LastCategoryName:Text=Get Name!(self:BP UI Habanero Tier Type Data Base Object Reference=@n4) -> (Name:Text)) -> (Output_Get:Text)
   n10 = Create Widget!(
    execute:Exec=[@n9.else, @n3.Completed]
    Class:User Widget Class Reference=/Game/UI/Frontend/Lobby/Habanero/WBP_Habanero_CategoryDisplay.WBP_Habanero_CategoryDisplay_C
    OwningPlayer:Player Controller Object Reference
   ) -> (then:Exec, ReturnValue:WBP Habanero Category Display Object Reference)
   > n11 = AddChildToHorizontalBox!(execute, self:Horizontal Box Object Reference=Get Categories(), Content:Widget Object Reference=@n10) -> (then, ReturnValue:Horizontal Box Slot Object Reference)
   > Set!(execute, Size_Value:float=1, Size_SizeRule:ESlateSizeRule Enum=Fill, self:Horizontal Box Slot Object Reference=@n11) -> (then, Output_Get:Slate Child Size Structure)
   > n12 = UpdateList!(
    execute
    self:WBP Habanero Category Display Object Reference=@n10
    Name:Text=Get LastCategoryName() -> (LastCategoryName:Text)
    List:Array of Fort Habanero Tier Structures=Get CategoryList()
   ) -> (then)
   Add(
    execute=@n5.then_1
    TargetArray:Array of Fort Habanero Tier Structures=Get CategoryList()
    NewItem:"Fort Habanero Tier Structure (by ref)"=@n3
   ) {is=title K2Node_CallArrayFunction}
   "To String (Integer)"(InInt=@n3.Array Index)
  ]}
 UpdateEndListPadding {nodes [
   n1 = UpdateEndListPadding!() -> (then, IsFirstEntry:bool, IsLastEntry:bool) {is=title K2Node_FunctionEntry}
   > SetSize!(
    execute
    self:Spacer Object Reference=Get Spacer_Top() -> (Spacer_Top:Spacer Object Reference)
    InSize_X:double=0
    InSize_Y:double=SelectFloat(A=200, B=0, bPickA=@n1)
   ) -> (then)
   > SetSize(self=Get Spacer_Bottom() -> (Spacer_Bottom:Spacer Object Reference), InSize_X=0, InSize_Y=SelectFloat(A=1000, B=0, bPickA=@n1.IsLastEntry))
   > Return Node()
  ]}
 UpdateFocus {nodes [
   n1 = UpdateFocus!() -> (then, IsFocused:bool) {is=title K2Node_FunctionEntry}
   > PlayAnimation(InAnimation=Select(
     Option 0:Widget Animation Object Reference=Get Unfocused() -> (Unfocused:Widget Animation Object Reference)
     Option 1:Widget Animation Object Reference=Get Focused() -> (Focused:Widget Animation Object Reference)
     Index=@n1
    ) -> (ReturnValue:Widget Animation Object Reference))
   > Return Node()
  ]}
 UpdateImage {nodes [UpdateImage!() -> (then, Object:Object Reference) {is=title K2Node_FunctionEntry}]}
 UpdatePanel {nodes [
   n1 = UpdatePanel!() -> (then, Object In:Object Reference) {is=title K2Node_FunctionEntry}
   > UpdateText!(execute, self, Object:Object Reference=@n1) -> (then)
   > Return Node()
  ]}
 UpdateText {nodes [
   n1 = UpdateText!() -> (then, Object:Object Reference) {is=title K2Node_FunctionEntry}
   > n2 = Cast To BP_UI_Habanero_InfoPanel!(execute, Object=@n1) -> (then, CastFailed, AsBP UI Habanero Info Panel:BP UI Habanero Info Panel Object Reference) {is=title K2Node_DynamicCast}
   > Reset!(
    execute
    self:Dynamic Entry Box Object Reference=Get BulletList() -> (BulletList:Dynamic Entry Box Object Reference)
    bDeleteWidgets:bool=false
   ) -> (then)
   > "SetText (Text)"(
    self=Get Text_Title() -> (Text_Title:Common Text Object Reference)
    InText=Get Title!(self:BP UI Habanero Info Panel Object Reference=@n2) -> (Title:Text)
   )
   > n3 = For Each Loop(Array:Array of Texts=Get Details List!(self:BP UI Habanero Info Panel Object Reference=@n2) -> (Details List:Array of Texts)) -> (Array Element:Text)
   > n4 = Create Entry!(execute, self:Dynamic Entry Box Object Reference=Get BulletList()) -> (then, ReturnValue:User Widget Object Reference)
   > n5 = Cast To GenericModalEntry!(execute, Object=@n4) -> (then, CastFailed, AsGeneric Modal Entry:Generic Modal Entry Object Reference) {is=title K2Node_DynamicCast}
   > SetText(self:Generic Modal Entry Object Reference=@n5, InText:Text=@n3)
  ]}
]
variables [
 RankCategories {type=EHabaneroTierType, default=HabaneroSauce0}
 RankData {type=FortHabaneroDisplayData, default=/Game/UI/Habanero/HabaneroDisplayData}
]
widgets {
 info {inheritedWidgetCount=0, parentClass=/Script/FortniteUI.FortHabaneroInfoEntry, rootWidgetClass=&Border, widgetCount=13}
 widgets [
  Border_0 {
   widgetClassPath=&Border
   widget=Border_0
   properties {
    background {
     bIsDynamicallyLoaded=false
     drawAs=NoDrawType
     imageType=NoImage
     mirroring=NoMirror
     resourceName=None
     resourceObject=null
     tiling=NoTile
     imageSize=<32 32>
     margin=<ltrb 0 0 0 0>
     outlineSettings {
      bUseBrushTransparency=false
      roundingType=HalfHeightRadius
      width=0
      color {colorUseRule=UseColor_Specified, specifiedColor=<rgba 0 0 0 0>}
      cornerRadii=<0 0 0 0>
     }
     tintColor {colorUseRule=UseColor_Specified, specifiedColor=<rgba 1 1 1 1>}
     uVRegion {bIsValid=false, max=<0 0>, min=<0 0>}
    }
    padding=<ltrb 0 0 0 0>
   }
  }
  VerticalBox_1 {widgetClassPath=&VerticalBox, parent=Border_0, widget=VerticalBox_1, slot {class=BorderSlot, properties {padding=<ltrb 0 0 0 0>}}}
  Spacer_Top {widgetClassPath=&Spacer, bIsVariable=true, parent=VerticalBox_1, widget=Spacer_Top, slot {class=VerticalBoxSlot}}
  Overlay_Modal {
   widgetClassPath=&Overlay
   parent=VerticalBox_1
   widget=Overlay_Modal
   slot {class=VerticalBoxSlot, properties {padding=<ltrb 0 32 0 32>, size {sizeRule=Fill, value=1}}}
  }
  PanelSize {
   widgetClassPath=&Spacer
   parent=Overlay_Modal
   widget=PanelSize
   properties {size=<600 32>}
   slot {class=OverlaySlot, properties {horizontalAlignment=HAlign_Fill, verticalAlignment=VAlign_Fill}}
  }
  Background {
   widgetClassPath=&Image
   bIsVariable=true
   parent=Overlay_Modal
   widget=Background
   properties {brush {
     bIsDynamicallyLoaded=false
     drawAs=Image
     imageType=NoImage
     mirroring=NoMirror
     resourceName=None
     resourceObject=/Game/UI/Frontend/Lobby/Habanero/Habanero_Materials/MaterialInstances/Background/MI_UI_Habanero_IntroModal_Bg
     tiling=NoTile
     imageSize=<32 32>
     margin=<ltrb 0 0 0 0>
     outlineSettings {
      bUseBrushTransparency=false
      roundingType=HalfHeightRadius
      width=0
      color {colorUseRule=UseColor_Specified, specifiedColor=<rgba 0 0 0 0>}
      cornerRadii=<0 0 0 0>
     }
     tintColor {colorUseRule=UseColor_Specified, specifiedColor=<rgba 1 1 1 1>}
     uVRegion {bIsValid=false, max=<0 0>, min=<0 0>}
    }}
   slot {class=OverlaySlot, properties {horizontalAlignment=HAlign_Fill, verticalAlignment=VAlign_Fill}}
  }
  VerticalBox_Content {
   widgetClassPath=&VerticalBox
   parent=Overlay_Modal
   widget=VerticalBox_Content
   slot {class=OverlaySlot, properties {horizontalAlignment=HAlign_Fill, verticalAlignment=VAlign_Fill}}
  }
  Categories {
   widgetClassPath=&HorizontalBox
   bIsVariable=true
   parent=VerticalBox_Content
   widget=Categories
   properties {flowDirectionPreference=Culture}
   slot {class=VerticalBoxSlot, properties {padding=<ltrb 16 0 16 0>}}
  }
  VerticalBox_Text {
   widgetClassPath=&VerticalBox
   parent=VerticalBox_Content
   widget=VerticalBox_Text
   slot {class=VerticalBoxSlot, properties {padding=<ltrb 48 16 48 32>}}
  }
  Text_Title {
   widgetClassPath=&CommonTextBlock
   bIsVariable=true
   parent=VerticalBox_Text
   widget=Text_Title
   properties {
    style=/FortUILibrary/Styles/Typography/_TextStyle_SubH4._TextStyle_SubH4_C
    text="<Title>"
    textTransformPolicy=ToUpper
    wrapTextAt=1152
    font {
     bForceMonospaced=false
     bMaterialIsStencil=false
     fontMaterial=null
     fontObject=&BurbankSmall
     letterSpacing=0
     monospacedWidth=1
     size=24
     skewAmount=0
     typefaceFontName=Black
     outlineSettings {
      bApplyOutlineToDropShadows=false
      bMiteredCorners=false
      bSeparateFillAlpha=false
      outlineMaterial=null
      outlineSize=0
      outlineColor=<rgba 0 0 0 1>
     }
    }
    shadowOffset=<0 0>
    strikeBrush {
     bIsDynamicallyLoaded=false
     drawAs=NoDrawType
     imageType=NoImage
     mirroring=NoMirror
     resourceName=None
     resourceObject=null
     tiling=NoTile
     imageSize=<32 32>
     margin=<ltrb 0 0 0 0>
     outlineSettings {
      bUseBrushTransparency=false
      roundingType=HalfHeightRadius
      width=0
      color {colorUseRule=UseColor_Specified, specifiedColor=<rgba 0 0 0 0>}
      cornerRadii=<0 0 0 0>
     }
     tintColor {colorUseRule=UseColor_Specified, specifiedColor=<rgba 1 1 1 1>}
     uVRegion {bIsValid=false, max=<0 0>, min=<0 0>}
    }
   }
   slot {class=VerticalBoxSlot}
  }
  TextContent {
   widgetClassPath=&VerticalBox
   parent=VerticalBox_Text
   widget=TextContent
   slot {class=VerticalBoxSlot, properties {padding=<ltrb 0 16 0 0>}}
  }
  BulletList {
   widgetClassPath=&DynamicEntryBox
   bIsVariable=true
   parent=TextContent
   widget=BulletList
   properties {entryBoxType=Vertical, entryWidgetClass=/Game/Athena/UI/GenericModalEntry.GenericModalEntry_C}
   slot {class=VerticalBoxSlot, properties {verticalAlignment=VAlign_Center, size {sizeRule=Fill, value=1}}}
  }
  Spacer_Bottom {widgetClassPath=&Spacer, bIsVariable=true, parent=VerticalBox_1, widget=Spacer_Bottom, slot {class=VerticalBoxSlot}}
 ]
}
