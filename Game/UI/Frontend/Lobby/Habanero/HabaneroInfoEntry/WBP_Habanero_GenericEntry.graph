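# Habanero lobby info panel entry (FortHabaneroGenericEntry child) showing a title, bullet list and optional right-side artwork.
# UpdatePanel reads a BP_UI_Habanero_InfoPanel object to build GenericModalEntry bullets, show or collapse icon layers, and tint glow and gradient-flare dynamic materials.
# UpdateFocus plays Focused/Unfocused fade-and-dim animations, while UpdateEndListPadding sizes top and bottom spacers for first and last entries.
# PreConstruct negates both flares' horizontal shear for Arabic to mirror the slant.
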WidgetBlueprint /Game/UI/Frontend/Lobby/Habanero/HabaneroInfoEntry/WBP_Habanero_GenericEntry
animations [
 Unfocused {
  binding_count=2
  length_seconds=0.2521
  object=&self:Unfocused
  sequence {bindings [
    HorizontalBox_0 {
     id=#1
     possessed_class=HorizontalBox
     tracks [{
       type=MovieSceneFloatTrack
       display_name=Render Opacity
       sections [{
         type=MovieSceneFloatSection
         end_frame=5
         start_frame=0
         channels [None {type=MovieSceneScriptingFloatChannel, keys [[0,1],[0.7497,0],[1.0113,0.0098]]}]
        }]
      }]
    }
    Border_0 {
     id=#2
     possessed_class=Border
     tracks [{
       type=MovieSceneColorTrack
       display_name=Content Color and Opacity
       sections [{
         type=MovieSceneColorSection
         end_frame=5
         start_frame=0
         channels [
          Color.R {type=MovieSceneScriptingFloatChannel, keys [[0,1],[5,0.1302]]}
          Color.G {type=MovieSceneScriptingFloatChannel, keys [[0,1],[5,0.1302]]}
          Color.B {type=MovieSceneScriptingFloatChannel, keys [[0,1],[5,0.1302]]}
          Color.A {type=MovieSceneScriptingFloatChannel, const [1,2]}
         ]
        }]
      }]
    }
   ]}
 }
 Focused {
  binding_count=2
  length_seconds=0.2521
  object=&self:Focused
  sequence {bindings [
    HorizontalBox_0 {
     id=#3
     possessed_class=HorizontalBox
     tracks [{
       type=MovieSceneFloatTrack
       display_name=Render Opacity
       sections [{type=MovieSceneFloatSection, end_frame=5, start_frame=0, channels [None {type=MovieSceneScriptingFloatChannel, keys [[0,0],[5,1]]}]}]
      }]
    }
    Border_0 {
     id=#2
     possessed_class=Border
     tracks [{
       type=MovieSceneColorTrack
       display_name=Content Color and Opacity
       sections [{
         type=MovieSceneColorSection
         end_frame=5
         start_frame=0
         channels [
          Color.R {type=MovieSceneScriptingFloatChannel, keys [[0,0.1302],[5,1],[5.0427,1]]}
          Color.G {type=MovieSceneScriptingFloatChannel, keys [[0,0.1302],[5,1],[5.0427,1]]}
          Color.B {type=MovieSceneScriptingFloatChannel, keys [[0,0.1302],[5,1],[5.0427,1]]}
          Color.A {type=MovieSceneScriptingFloatChannel, const [1,3]}
         ]
        }]
      }]
    }
   ]}
 }
]
graphs [
 EventGraph {nodes [
   Event PreConstruct()
   n1 = Get GradientFlare() -> (GradientFlare:Image Object Reference)
   n2 = Break Widget Transform!(WidgetTransform:"Widget Transform Structure (by ref)"=Get RenderTransform!(self:Widget Object Reference=@n1) -> (RenderTransform:Widget Transform Structure)) -> (Translation:Vector 2D Structure, Scale:Vector 2D Structure, Shear:Vector 2D Structure, Angle:float=~) {is=title K2Node_BreakStruct}
   n3 = "Equal, Case Insensitive (String)"(A=GetCurrentLanguage!() -> (ReturnValue:String), B=ar)
   n4 = BreakVector2D(InVec=@n2.Shear)
   n5 = float * float(A=@n4, B=-1) {is=title K2Node_PromotableOperator}
   > SetRenderShear!(
    execute
    self:Widget Object Reference=@n1
    Shear:Vector 2D Structure=Select(Option 0:Vector 2D Structure=@n2.Shear, Option 1:Vector 2D Structure=MakeVector2D(X=@n5, Y=@n4.Y), Index=@n3) -> (ReturnValue:Vector 2D Structure)
   ) -> (then)
   n6 = Get GradientFlare_Bg() -> (GradientFlare_Bg:Image Object Reference)
   n7 = Break Widget Transform(WidgetTransform=Get RenderTransform(self=@n6)) -> (Angle=~)
   n8 = BreakVector2D(InVec=@n7.Shear)
   n9 = float * float(A=@n8, B=-1)
   > SetRenderShear(self=@n6, Shear=Select(Option 0=@n7.Shear, Option 1=MakeVector2D(X=@n9, Y=@n8.Y), Index=@n3))
  ]}
 UpdateBullets {nodes [
   n1 = UpdateBullets!() -> (then, Object:Object Reference) {is=title K2Node_FunctionEntry}
   > n2 = Cast To BP_UI_Habanero_InfoPanel!(execute, Object=@n1) -> (then, CastFailed, AsBP UI Habanero Info Panel:BP UI Habanero Info Panel Object Reference) {is=title K2Node_DynamicCast}
   > Reset!(
    execute
    self:Dynamic Entry Box Object Reference=Get BulletList() -> (BulletList:Dynamic Entry Box Object Reference)
    bDeleteWidgets:bool=false
   ) -> (then)
   > "SetText (Text)"(
    self=Get Text_Title() -> (Text_Title:Common Text Object Reference)
    InText=Get Title!(self:BP UI Habanero Info Panel Object Reference=@n2) -> (Title:Text)
   )
   > n3 = For Each Loop(Array:Array of Texts=Get Details List!(self:BP UI Habanero Info Panel Object Reference=@n2) -> (Details List:Array of Texts)) -> (Array Element:Text)
   > n4 = Create Entry!(execute, self:Dynamic Entry Box Object Reference=Get BulletList()) -> (then, ReturnValue:User Widget Object Reference)
   > n5 = Cast To GenericModalEntry!(execute, Object=@n4) -> (then, CastFailed, AsGeneric Modal Entry:Generic Modal Entry Object Reference) {is=title K2Node_DynamicCast}
   > SetText(self:Generic Modal Entry Object Reference=@n5, InText:Text=@n3)
  ]}
 UpdateEndListPadding {nodes [
   n1 = UpdateEndListPadding!() -> (then, IsFirstEntry:bool, IsLastEntry:bool) {is=title K2Node_FunctionEntry}
   > SetSize!(
    execute
    self:Spacer Object Reference=Get Spacer_Top() -> (Spacer_Top:Spacer Object Reference)
    InSize_X:double=0
    InSize_Y:double=SelectFloat(A=256, B=0, bPickA=@n1)
   ) -> (then)
   > SetSize(self=Get Spacer_Bottom() -> (Spacer_Bottom:Spacer Object Reference), InSize_X=0, InSize_Y=SelectFloat(A=1000, B=0, bPickA=@n1.IsLastEntry))
   > Return Node()
  ]}
 UpdateFocus {nodes [
   n1 = UpdateFocus!() -> (then, IsFocused:bool) {is=title K2Node_FunctionEntry}
   > n2 = Branch(Condition=IsAnyAnimationPlaying!(self:User Widget Object Reference) -> (ReturnValue:bool))
   > n3 = StopAllAnimations(self:User Widget Object Reference)
   PlayAnimation(
    execute=[@n3.then, @n2.else]
    InAnimation=Select(
     Option 0:Widget Animation Object Reference=Get Unfocused() -> (Unfocused:Widget Animation Object Reference)
     Option 1:Widget Animation Object Reference=Get Focused() -> (Focused:Widget Animation Object Reference)
     Index=@n1
    ) -> (ReturnValue:Widget Animation Object Reference)
   )
   > Return Node()
  ]}
 UpdateImage {nodes [
   n1 = UpdateImage!() -> (then, Object:Object Reference) {is=title K2Node_FunctionEntry}
   > n2 = Cast To BP_UI_Habanero_InfoPanel(Object=@n1)
   n3 = Get Image!(self:BP UI Habanero Info Panel Object Reference=@n2) -> (Image:Material Instance Object Reference)
   > n4 = Branch(Condition=IsValid(Object=@n3))
   > SetBrushFromMaterial(self=Get Icon() -> (Icon:Image Object Reference), Material=@n3)
   n5 = Get Image_Content_R() -> (Image_Content_R:Scale Box Object Reference)
   > SetVisibility(self=@n5, InVisibility=HitTestInvisible)
   n6 = Get ImageBG!(self:BP UI Habanero Info Panel Object Reference=@n2) -> (ImageBG:Material Instance Object Reference)
   > n7 = Branch(Condition=IsValid(Object=@n6))
   n8 = Get IconBG() -> (IconBG:Image Object Reference)
   > SetBrushFromMaterial(self=@n8, Material=@n6)
   > SetVisibility(self=@n8, InVisibility=HitTestInvisible)
   SetVisibility(execute=@n7.else, self=Get IconBG(), InVisibility=Collapsed)
   SetVisibility(execute=@n4.else, self=@n5, InVisibility=Collapsed)
  ]}
 UpdatePanel {nodes [
   n1 = UpdatePanel!() -> (then, Object In:Object Reference) {is=title K2Node_FunctionEntry}
   > UpdateBullets!(execute, self, Object:Object Reference=@n1) -> (then) {is=title K2Node_CallFunction}
   > UpdateImage!(execute, self, Object:Object Reference=@n1) -> (then) {is=title K2Node_CallFunction}
   > UpdatePanelColor!(execute, self, Object:Object Reference=@n1) -> (then)
   > Return Node()
  ]}
 UpdatePanelColor {nodes [
   n1 = UpdatePanelColor!() -> (then, Object:Object Reference) {is=title K2Node_FunctionEntry}
   > n2 = Cast To BP_UI_Habanero_InfoPanel(Object=@n1)
   n3 = Break Habanero Panel Colors!(Habanero_PanelColors:"Habanero Panel Colors Structure (by ref)"=Get PanelColors!(self:BP UI Habanero Info Panel Object Reference=@n2) -> (PanelColors:Habanero Panel Colors Structure)) -> (
    Glow_19_#4:Linear Color Structure
    Flare_ColorA_14_#5:Linear Color Structure
    Flare_ColorB_15_#6:Linear Color Structure
    FlareBG_ColorA_16_#7:Linear Color Structure
    FlareBG_ColorB_17_#8:Linear Color Structure
   ) {is=title K2Node_BreakStruct}
   > SetColorAndOpacity(self:Image Object Reference=Get Glow() -> (Glow:Image Object Reference), InColorAndOpacity:Linear Color Structure=@n3)
   > n4 = GetDynamicMaterial(self=Get GradientFlare())
   > SetVectorParameterValue(self=@n4, ParameterName=Color A, Value=@n3.Flare_ColorA_14_#5)
   > SetVectorParameterValue(self=@n4, ParameterName=Color B, Value=@n3.Flare_ColorB_15_#6)
   > n5 = GetDynamicMaterial(self=Get GradientFlare_Bg())
   > SetVectorParameterValue(self=@n5, ParameterName=Color A, Value=@n3.FlareBG_ColorA_16_#7)
   > SetVectorParameterValue(self=@n5, ParameterName=Color B, Value=@n3.FlareBG_ColorB_17_#8)
  ]}
]
variables []
widgets {
 info {inheritedWidgetCount=0, parentClass=/Script/FortniteUI.FortHabaneroGenericEntry, rootWidgetClass=&Border, widgetCount=18}
 widgets [
  Border_0 {
   widgetClassPath=&Border
   widget=Border_0
   properties {
    background {
     bIsDynamicallyLoaded=false
     drawAs=NoDrawType
     imageType=NoImage
     mirroring=NoMirror
     resourceName=None
     resourceObject=null
     tiling=NoTile
     imageSize=<32 32>
     margin=<ltrb 0 0 0 0>
     outlineSettings {
      bUseBrushTransparency=false
      roundingType=HalfHeightRadius
      width=0
      color {colorUseRule=UseColor_Specified, specifiedColor=<rgba 0 0 0 0>}
      cornerRadii=<0 0 0 0>
     }
     tintColor {colorUseRule=UseColor_Specified, specifiedColor=<rgba 1 1 1 1>}
     uVRegion {bIsValid=false, max=<0 0>, min=<0 0>}
    }
    padding=<ltrb 0 32 0 32>
   }
  }
  VerticalBox_1 {
   widgetClassPath=&VerticalBox
   parent=Border_0
   widget=VerticalBox_1
   slot {class=BorderSlot, properties {padding=<ltrb 0 32 0 32>}}
  }
  Spacer_Top {widgetClassPath=&Spacer, bIsVariable=true, parent=VerticalBox_1, widget=Spacer_Top, slot {class=VerticalBoxSlot}}
  Overlay_Modal {
   widgetClassPath=&Overlay
   parent=VerticalBox_1
   widget=Overlay_Modal
   slot {class=VerticalBoxSlot, properties {size {sizeRule=Fill, value=1}}}
  }
  PanelSize {
   widgetClassPath=&Spacer
   parent=Overlay_Modal
   widget=PanelSize
   properties {size=<664 664>}
   slot {class=OverlaySlot, properties {horizontalAlignment=HAlign_Fill, verticalAlignment=VAlign_Fill}}
  }
  Background {
   widgetClassPath=&Image
   bIsVariable=true
   parent=Overlay_Modal
   widget=Background
   properties {
    visibility=SelfHitTestInvisible
    brush {
     bIsDynamicallyLoaded=false
     drawAs=Image
     imageType=NoImage
     mirroring=NoMirror
     resourceName=None
     resourceObject=/Game/UI/Frontend/Lobby/Habanero/Habanero_Materials/MaterialInstances/Background/MI_UI_Habanero_IntroModal_Bg
     tiling=NoTile
     imageSize=<32 32>
     margin=<ltrb 0 0 0 0>
     outlineSettings {
      bUseBrushTransparency=false
      roundingType=HalfHeightRadius
      width=0
      color {colorUseRule=UseColor_Specified, specifiedColor=<rgba 0 0 0 0>}
      cornerRadii=<0 0 0 0>
     }
     tintColor {colorUseRule=UseColor_Specified, specifiedColor=<rgba 1 1 1 1>}
     uVRegion {bIsValid=false, max=<0 0>, min=<0 0>}
    }
   }
   slot {class=OverlaySlot, properties {horizontalAlignment=HAlign_Fill, verticalAlignment=VAlign_Fill}}
  }
  HorizontalBox_0 {
   widgetClassPath=&HorizontalBox
   parent=Overlay_Modal
   widget=HorizontalBox_0
   properties {clipping=ClipToBounds, flowDirectionPreference=Culture}
   slot {class=OverlaySlot, properties {horizontalAlignment=HAlign_Fill, verticalAlignment=VAlign_Fill}}
  }
  VerticalBox_0 {
   widgetClassPath=&VerticalBox
   parent=HorizontalBox_0
   widget=VerticalBox_0
   slot {class=HorizontalBoxSlot, properties {padding=<ltrb 80 80 92 35>, size {sizeRule=Fill, value=1}}}
  }
  Text_Title {
   widgetClassPath=&CommonTextBlock
   bIsVariable=true
   parent=VerticalBox_0
   widget=Text_Title
   properties {
    style=/FortUILibrary/Styles/Typography/_TextStyle_SubH4._TextStyle_SubH4_C
    text="<Title>"
    textTransformPolicy=ToUpper
    wrapTextAt=1152
    font {
     bForceMonospaced=false
     bMaterialIsStencil=false
     fontMaterial=null
     fontObject=&BurbankSmall
     letterSpacing=0
     monospacedWidth=1
     size=24
     skewAmount=0
     typefaceFontName=Black
     outlineSettings {
      bApplyOutlineToDropShadows=false
      bMiteredCorners=false
      bSeparateFillAlpha=false
      outlineMaterial=null
      outlineSize=0
      outlineColor=<rgba 0 0 0 1>
     }
    }
    shadowOffset=<0 0>
    strikeBrush {
     bIsDynamicallyLoaded=false
     drawAs=NoDrawType
     imageType=NoImage
     mirroring=NoMirror
     resourceName=None
     resourceObject=null
     tiling=NoTile
     imageSize=<32 32>
     margin=<ltrb 0 0 0 0>
     outlineSettings {
      bUseBrushTransparency=false
      roundingType=HalfHeightRadius
      width=0
      color {colorUseRule=UseColor_Specified, specifiedColor=<rgba 0 0 0 0>}
      cornerRadii=<0 0 0 0>
     }
     tintColor {colorUseRule=UseColor_Specified, specifiedColor=<rgba 1 1 1 1>}
     uVRegion {bIsValid=false, max=<0 0>, min=<0 0>}
    }
   }
   slot {class=VerticalBoxSlot, properties {padding=<ltrb 0 0 0 16>}}
  }
  BulletList {
   widgetClassPath=&DynamicEntryBox
   bIsVariable=true
   parent=VerticalBox_0
   widget=BulletList
   properties {entryBoxType=Vertical, entryWidgetClass=/Game/Athena/UI/GenericModalEntry.GenericModalEntry_C}
   slot {class=VerticalBoxSlot, properties {size {sizeRule=Fill, value=1}}}
  }
  Image_Content_R {
   widgetClassPath=&ScaleBox
   bIsVariable=true
   parent=HorizontalBox_0
   widget=Image_Content_R
   properties {stretchDirection=DownOnly, visibility=Collapsed}
   slot {class=HorizontalBoxSlot}
  }
  Overlay_0 {
   widgetClassPath=&Overlay
   parent=Image_Content_R
   widget=Overlay_0
   slot {class=ScaleBoxSlot, properties {horizontalAlignment=HAlign_Fill, verticalAlignment=VAlign_Fill}}
  }
  Glow {
   widgetClassPath=&Image
   bIsVariable=true
   parent=Overlay_0
   widget=Glow
   properties {
    visibility=HitTestInvisible
    brush {
     bIsDynamicallyLoaded=false
     drawAs=Image
     imageType=NoImage
     mirroring=NoMirror
     resourceName=None
     resourceObject=/Game/UI/Frontend/Lobby/Habanero/Habanero_Materials/MaterialInstances/Flares/MI_UI_Habanero_RadialGlow
     tiling=NoTile
     imageSize=<32 32>
     margin=<ltrb 0 0 0 0>
     outlineSettings {
      bUseBrushTransparency=false
      roundingType=HalfHeightRadius
      width=0
      color {colorUseRule=UseColor_Specified, specifiedColor=<rgba 0 0 0 0>}
      cornerRadii=<0 0 0 0>
     }
     tintColor {colorUseRule=UseColor_Specified, specifiedColor=<rgba 1 1 1 1>}
     uVRegion {bIsValid=false, max=<0 0>, min=<0 0>}
    }
    colorAndOpacity=<rgba 0.1095 0.016 1 1>
    renderTransform {angle=0, scale=<32 32>, shear=<0 0>, translation=<0 0>}
   }
   slot {class=OverlaySlot, properties {horizontalAlignment=HAlign_Right}}
  }
  GradientFlare_Bg {
   widgetClassPath=&Image
   bIsVariable=true
   parent=Overlay_0
   widget=GradientFlare_Bg
   properties {
    brush {
     bIsDynamicallyLoaded=false
     drawAs=Image
     imageType=NoImage
     mirroring=NoMirror
     resourceName=None
     resourceObject=/Game/UI/Frontend/Lobby/Habanero/Habanero_Materials/MaterialInstances/Background/MI_UI_Habanero_IntroModal_FlareGradient_Secondary
     tiling=NoTile
     imageSize=<32 32>
     margin=<ltrb 0 0 0 0>
     outlineSettings {
      bUseBrushTransparency=false
      roundingType=HalfHeightRadius
      width=0
      color {colorUseRule=UseColor_Specified, specifiedColor=<rgba 0 0 0 0>}
      cornerRadii=<0 0 0 0>
     }
     tintColor {colorUseRule=UseColor_Specified, specifiedColor=<rgba 1 1 1 1>}
     uVRegion {bIsValid=false, max=<0 0>, min=<0 0>}
    }
    renderTransform {angle=0, scale=<1 1>, shear=<24 0>, translation=<0 0>}
   }
   slot {class=OverlaySlot, properties {horizontalAlignment=HAlign_Fill, verticalAlignment=VAlign_Fill}}
  }
  GradientFlare {
   widgetClassPath=&Image
   bIsVariable=true
   parent=Overlay_0
   widget=GradientFlare
   properties {
    brush {
     bIsDynamicallyLoaded=false
     drawAs=Image
     imageType=NoImage
     mirroring=NoMirror
     resourceName=None
     resourceObject=/Game/UI/Frontend/Lobby/Habanero/Habanero_Materials/MaterialInstances/Background/MI_UI_Habanero_IntroModal_FlareGradient
     tiling=NoTile
     imageSize=<32 32>
     margin=<ltrb 0 0 0 0>
     outlineSettings {
      bUseBrushTransparency=false
      roundingType=HalfHeightRadius
      width=0
      color {colorUseRule=UseColor_Specified, specifiedColor=<rgba 0 0 0 0>}
      cornerRadii=<0 0 0 0>
     }
     tintColor {colorUseRule=UseColor_Specified, specifiedColor=<rgba 1 1 1 1>}
     uVRegion {bIsValid=false, max=<0 0>, min=<0 0>}
    }
    renderTransform {angle=0, scale=<0.875 1>, shear=<24 0>, translation=<0 0>}
   }
   slot {class=OverlaySlot, properties {horizontalAlignment=HAlign_Fill, verticalAlignment=VAlign_Fill}}
  }
  IconBG {
   widgetClassPath=&Image
   bIsVariable=true
   parent=Overlay_0
   widget=IconBG
   properties {
    visibility=HitTestInvisible
    brush {
     bIsDynamicallyLoaded=false
     drawAs=Image
     imageType=NoImage
     mirroring=NoMirror
     resourceName=None
     resourceObject=/Game/UI/Frontend/Lobby/Habanero/Habanero_Materials/MaterialInstances/Icons/Foreground/MI_UI_Habanero_Foreground_00
     tiling=NoTile
     imageSize=<600 600>
     margin=<ltrb 0 0 0 0>
     outlineSettings {
      bUseBrushTransparency=false
      roundingType=HalfHeightRadius
      width=0
      color {colorUseRule=UseColor_Specified, specifiedColor=<rgba 0 0 0 0>}
      cornerRadii=<0 0 0 0>
     }
     tintColor {colorUseRule=UseColor_Specified, specifiedColor=<rgba 1 1 1 1>}
     uVRegion {bIsValid=false, max=<0 0>, min=<0 0>}
    }
   }
   slot {class=OverlaySlot, properties {verticalAlignment=VAlign_Center, padding=<ltrb 32 32 32 32>}}
  }
  Icon {
   widgetClassPath=&Image
   bIsVariable=true
   parent=Overlay_0
   widget=Icon
   properties {
    visibility=HitTestInvisible
    brush {
     bIsDynamicallyLoaded=false
     drawAs=Image
     imageType=NoImage
     mirroring=NoMirror
     resourceName=None
     resourceObject=&MI_UI_Habanero_Foreground_00
     tiling=NoTile
     imageSize=<600 600>
     margin=<ltrb 0 0 0 0>
     outlineSettings {
      bUseBrushTransparency=false
      roundingType=HalfHeightRadius
      width=0
      color {colorUseRule=UseColor_Specified, specifiedColor=<rgba 0 0 0 0>}
      cornerRadii=<0 0 0 0>
     }
     tintColor {colorUseRule=UseColor_Specified, specifiedColor=<rgba 1 1 1 1>}
     uVRegion {bIsValid=false, max=<0 0>, min=<0 0>}
    }
   }
   slot {class=OverlaySlot, properties {verticalAlignment=VAlign_Center, padding=<ltrb 32 32 32 32>}}
  }
  Spacer_Bottom {widgetClassPath=&Spacer, bIsVariable=true, parent=VerticalBox_1, widget=Spacer_Bottom, slot {class=VerticalBoxSlot}}
 ]
}
