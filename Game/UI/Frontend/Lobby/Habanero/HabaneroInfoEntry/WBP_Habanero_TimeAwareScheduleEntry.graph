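# Lobby info entry, derived from the generic Habanero entry, listing when Ranked playlists are enabled per region for Battle Royale Build and Zero Build.
# UpdateBullets feeds the generated text array into the info panel's details list before calling the parent.
# Data-table UTC hours are shifted by the local Now-minus-UTC offset, with day names adjusted across midnight.
# The header names the time zone, using localized names for select offset-and-language pairs, otherwise a UTC offset.
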WidgetBlueprint /Game/UI/Frontend/Lobby/Habanero/HabaneroInfoEntry/WBP_Habanero_TimeAwareScheduleEntry
graphs [
 EventGraph {nodes [
   n1 = Event UpdateBullets!() -> (OutputDelegate, then, Object:Object Reference) {is=title K2Node_Event}
   > n2 = Cast To BP_UI_Habanero_InfoPanel!(execute, Object=@n1.Object) -> (then, CastFailed, AsBP UI Habanero Info Panel:BP UI Habanero Info Panel Object Reference) {is=title K2Node_DynamicCast}
   > Set Details List!(
    execute
    Details List:Array of Texts=MakeDynamicTimeAwareText!() -> (Array:Array of Texts)
    self:BP UI Habanero Info Panel Object Reference=@n2
   ) -> (then, Output_Get:Array of Texts)
   > Parent: UpdateBullets!(execute, Object:Object Reference=@n2) -> (then) {is=title K2Node_CallParentFunction}
  ]}
 FormatAvailableTimeText {nodes [
   n1 = FormatAvailableTimeText!() -> (then, MonFriUTCStartHour:int, MonFriUTCEndHour:int, SatSunUTCStartHour:int, SatSunUTCEndHour:int) {is=title K2Node_FunctionEntry}
   n2 = FormatHours!(UTCHour:int=@n1.SatSunUTCEndHour) -> (ResultValue:Text, NextDay:bool, PreviousDay:bool)
   n3 = FormatHours(UTCHour=@n1)
   n4 = FormatHours(UTCHour=@n1.MonFriUTCEndHour)
   n5 = FormatHours(UTCHour=@n1.SatSunUTCStartHour)
   n6 = Select(Option 0:Text="LOCTEXT(\"Friday\")", Option 1:Text="LOCTEXT(\"Thursday\")", Index=@n4.PreviousDay) -> (ReturnValue:Text)
   n7 = Select(Option 0="LOCTEXT(\"Sunday\")", Option 1="LOCTEXT(\"Saturday\")", Index=@n2.PreviousDay)
   n8 = Select(Option 0="LOCTEXT(\"Monday\")", Option 1="LOCTEXT(\"Sunday\")", Index=@n3.PreviousDay)
   n9 = Select(Option 0="LOCTEXT(\"Saturday\")", Option 1="LOCTEXT(\"Friday\")", Index=@n5.PreviousDay)
   > Return Node!(
    execute
    ReturnValue:Text=Format Text!(
     Format="LOCTEXT(\"{D1} - {D2}: {MonFriStart} - {MonFriEnd} <b>|</> {D3} - {D4}: {SatSunStart} - {SatSunEnd}\")"
     D1:Text=Select(Option 0=@n8, Option 1="LOCTEXT(\"Tuesday\")", Index=@n3.NextDay)
     D2:Text=Select(Option 0=@n6, Option 1="LOCTEXT(\"Saturday\")", Index=@n4.NextDay)
     MonFriStart:Text=@n3
     MonFriEnd:Text=@n4
     D3:Text=Select(Option 0=@n9, Option 1="LOCTEXT(\"Sunday\")", Index=@n5.NextDay)
     D4:Text=Select(Option 0=@n7, Option 1="LOCTEXT(\"Monday\")", Index=@n2.NextDay)
     SatSunStart:Text=@n5
     SatSunEnd:Text=@n2
    ) -> (Result)
   )
  ]}
 FormatBulletPointTextForRegionFromTable {nodes [
   n1 = FormatBulletPointTextForRegionFromTable!() -> (then, RegionName:Text, Type:Text, RegionDataTableRowName:Name) {is=title K2Node_FunctionEntry}
   n2 = FormatHoursFromDataTable!(Region:Name=@n1.RegionDataTableRowName) -> (ReturnValue:Text)
   > Return Node!(
    execute
    Result:Text=Select(
     Option 0=Format Text!(
      Format="LOCTEXT(\"<b>{RegionName}:</> {Type} - {RegionAvailableTime}\")"
      RegionName:Text=@n1
      Type:Text=@n1.Type
      RegionAvailableTime:Text=@n2
     ) -> (Result)
     Option 1=Format Text!(Format="LOCTEXT(\"<b>{RegionName}:</> {RegionAvailableTime}\")", RegionName=@n1, RegionAvailableTime=@n2) -> (Result)
     Index=TextIsEmpty(InText=@n1.Type)
    )
   )
  ]}
 FormatHours {nodes [
   n1 = FormatHours!() -> (then, UTCHour:int) {is=title K2Node_FunctionEntry}
   n2 = "To String (Integer)"(InInt=@n1)
   n3 = DateTime + Timespan!(
    A:Date Time Structure=DateTimeFromIsoString!(IsoString:String=Append!(A=2023-01-02T, B=@n2, C:String=:00:00+00:00) -> (ReturnValue)) -> (Result:Date Time Structure, ReturnValue:bool)
    B:Timespan Structure=DateTime - DateTime!(A:Date Time Structure=Now!() -> (ReturnValue:Date Time Structure), B:Date Time Structure=UTC Now!() -> (ReturnValue:Date Time Structure)) -> (ReturnValue:Timespan Structure)
   ) -> (ReturnValue:Date Time Structure)
   n4 = Get Day!(A:Date Time Structure=@n3) -> (ReturnValue:int=0) {is=title K2Node_CallFunction}
   > Return Node!(
    execute
    ResultValue:Text=As Time!(In:"Date Time Structure (by ref)"=@n3, InTimeStyle:EDateTimeStyle Enum=Short) -> (ReturnValue:Text)
    NextDay:bool="Equal (Integer)"(A=@n4, B=3)
    PreviousDay:bool="Equal (Integer)"(A=@n4, B=1)
   )
  ]}
 FormatHoursFromDataTable {nodes [
   n1 = FormatHoursFromDataTable!() -> (then, Region:Name) {is=title K2Node_FunctionEntry}
   > n2 = Get Data Table Row Habanero_ScheduleEntryDataTable!(
    execute:Exec
    DataTable:Data Table Object Reference=/Game/UI/Frontend/Lobby/Habanero/HabaneroData/Habanero_ScheduleEntryDataTable
    RowName:Name=@n1
   ) -> (then:Exec, RowNotFound:Exec, ReturnValue:Habanero Schedule Entry Data Structure Structure) {is=title K2Node_GetDataTableRow}
   n3 = Break Habanero Schedule Entry Data Structure!(Habanero_ScheduleEntryDataStructure:"Habanero Schedule Entry Data Structure Structure (by ref)"=@n2) -> (
    MonFriStartHour_4_#1:int
    MonFriEndHour_6_#2:int
    SatSunStartHour_8_#3:int
    SatSunEndHour_10_#4:int
   ) {is=title K2Node_BreakStruct}
   > n4 = Branch(Condition=AND Boolean!(
     A="Equal (Integer)"(A=@n3)
     B="Equal (Integer)"(A=@n3.MonFriEndHour_6_#2)
     C:bool="Equal (Integer)"(A=@n3.SatSunStartHour_8_#3)
     D:bool="Equal (Integer)"(A=@n3.SatSunEndHour_10_#4)
    ) -> (ReturnValue))
   > Return Node!(execute, ReturnValue:Text="LOCTEXT(\"All hours of the week\")")
   n5 = Break Habanero Schedule Entry Data Structure(Habanero_ScheduleEntryDataStructure=@n2)
   Return Node(
    execute=@n4.else
    ReturnValue=FormatAvailableTimeText!(
     MonFriUTCStartHour:int=@n5
     MonFriUTCEndHour:int=@n5.MonFriEndHour_6_#2
     SatSunUTCStartHour:int=@n5.SatSunStartHour_8_#3
     SatSunUTCEndHour:int=@n5.SatSunEndHour_10_#4
    ) -> (ReturnValue:Text) {is=title K2Node_CallFunction}
   )
   Return Node(execute=@n2.RowNotFound, ReturnValue="LOCTEXT(\"Unavailable\")")
  ]}
 FormatTimeZoneIdentifierText {nodes [
   FormatTimeZoneIdentifierText() {is=title K2Node_FunctionEntry}
   > Set TimeZoneOffsetUTC(TimeZoneOffsetUTC:Timespan Structure=DateTime - DateTime(A=Now(), B=UTC Now())) -> (Output_Get:Timespan Structure)
   n1 = Left!(SourceString:String=GetCurrentLanguage!() -> (ReturnValue:String), Count:int=2) -> (ReturnValue:String)
   n2 = Get Hours!(A:Timespan Structure=Get TimeZoneOffsetUTC() -> (TimeZoneOffsetUTC:Timespan Structure)) -> (ReturnValue:int=0) {is=title K2Node_CallFunction}
   > n3 = Branch(Condition=AND Boolean!(A="Equal (Integer)"(A=@n2, B=-8), B="Equal, Case Insensitive (String)"(A=@n1, B=en)) -> (ReturnValue))
   > Return Node(ReturnValue="INVTEXT(\"PST\")")
   n4 = Left(SourceString=GetCurrentLanguage(), Count=2)
   n5 = Get Hours(A=Get TimeZoneOffsetUTC()) -> (ReturnValue=0)
   n6 = Branch(execute=@n3.else, Condition=AND Boolean(A="Equal (Integer)"(A=@n5, B=-7), B="Equal, Case Insensitive (String)"(A=@n4, B=en)))
   > Return Node(ReturnValue="INVTEXT(\"MST\")")
   n7 = Left(SourceString=GetCurrentLanguage(), Count=2)
   n8 = Get Hours(A=Get TimeZoneOffsetUTC()) -> (ReturnValue=0)
   n9 = Branch(execute=@n6.else, Condition=AND Boolean(A="Equal (Integer)"(A=@n8, B=-6), B="Equal, Case Insensitive (String)"(A=@n7, B=en)))
   > Return Node(ReturnValue="INVTEXT(\"CST\")")
   n10 = Left(SourceString=GetCurrentLanguage(), Count=2)
   n11 = Get Hours(A=Get TimeZoneOffsetUTC()) -> (ReturnValue=0)
   n12 = Branch(execute=@n9.else, Condition=AND Boolean(A="Equal (Integer)"(A=@n11, B=-5), B="Equal, Case Insensitive (String)"(A=@n10, B=en)))
   > Return Node(ReturnValue="INVTEXT(\"EST\")")
   n13 = Left(SourceString=GetCurrentLanguage(), Count=2)
   n14 = Get Hours(A=Get TimeZoneOffsetUTC()) -> (ReturnValue=0)
   n15 = Branch(execute=@n12.else, Condition=AND Boolean(A="Equal (Integer)"(A=@n14, B=-3), B="Equal, Case Insensitive (String)"(A=@n13, B=pt)))
   > Return Node(ReturnValue="INVTEXT(\"Horário de Brasília\")")
   n16 = Get TimeZoneOffsetUTC()
   n17 = Branch(
    execute=@n15.else
    Condition="integer < integer"(A=Get Minutes!(A:Timespan Structure=@n16) -> (ReturnValue:int=0) {is=title K2Node_CallFunction}, B=12)
   )
   n18 = MakeLiteralText(Value="LOCTEXT(\"UTC{+0}\")")
   n19 = Get Hours(A=@n16) -> (ReturnValue=0)
   n20 = "To Text (Integer)"(Value=@n19, bAlwaysSign=true, bUseGrouping=false)
   n21 = "Equal (Integer)"(A=@n19)
   > Return Node(ReturnValue=Format Text!(Format=@n18, +0:Text=Select(Option 0=@n20, Index=@n21)) -> (Result))
   n22 = Get TimeZoneOffsetUTC()
   n23 = Get Minutes(A=@n22) -> (ReturnValue=0)
   n24 = float + float(A=float / float(A="To Float (Integer)"(InInt=@n23), B=60), B:int=Get Hours(A=@n22) -> (ReturnValue=0))
   Return Node(
    execute=@n17.else
    ReturnValue=Format Text(
     Format=@n18
     +0="To Text (Float)"(
      Value=@n24
      bAlwaysSign=true
      bUseGrouping=false
      MinimumIntegralDigits=1
      MaximumIntegralDigits=324
      MinimumFractionalDigits=0
      MaximumFractionalDigits=1
     )
    )
   )
  ]}
 MakeDynamicTimeAwareText {nodes [
   MakeDynamicTimeAwareText() {is=title K2Node_FunctionEntry}
   n1 = FormatBulletPointTextForRegionFromTable!(RegionName:Text="LOCTEXT(\"Middle East\")", Type:Text="LOCTEXT(\"Squads\")", RegionDataTableRowName:Name=zb-middle-east) -> (Result:Text) {is=title K2Node_CallFunction}
   n2 = FormatBulletPointTextForRegionFromTable(RegionName="LOCTEXT(\"NA-East, NA-Central, NA-West and Europe\")", RegionDataTableRowName=zb-na-east-and-europe)
   n3 = FormatBulletPointTextForRegionFromTable(RegionName="LOCTEXT(\"Oceania\")", Type="LOCTEXT(\"Squads\")", RegionDataTableRowName=zb-oceania)
   n4 = FormatBulletPointTextForRegionFromTable(RegionName="LOCTEXT(\"Brazil\")", Type="LOCTEXT(\"Squads \")", RegionDataTableRowName=zb-brazil)
   n5 = FormatBulletPointTextForRegionFromTable(RegionName="LOCTEXT(\"Asia\")", Type="LOCTEXT(\"Squads\")", RegionDataTableRowName=zb-asia)
   n6 = FormatBulletPointTextForRegionFromTable(RegionName="LOCTEXT(\"Brazil\")", Type="LOCTEXT(\"Squads\")", RegionDataTableRowName=brazil)
   n7 = FormatBulletPointTextForRegionFromTable(RegionName="LOCTEXT(\"Asia\")", Type="LOCTEXT(\"Squads\")", RegionDataTableRowName=asia)
   n8 = FormatBulletPointTextForRegionFromTable(RegionName="LOCTEXT(\"Middle East\")", Type="LOCTEXT(\"Squads\")", RegionDataTableRowName=middle-east)
   n9 = FormatBulletPointTextForRegionFromTable(RegionName="LOCTEXT(\"NA-East, NA-Central, NA-West and Europe\")", RegionDataTableRowName=na-east-and-europe)
   n10 = FormatBulletPointTextForRegionFromTable(RegionName="LOCTEXT(\"Oceania\")", Type="LOCTEXT(\"Solo/Duo\")", RegionDataTableRowName=oceania)
   n11 = FormatTimeZoneIdentifierText!() -> (ReturnValue:Text) {is=title K2Node_CallFunction}
   > Return Node!(
    execute
    Array:Array of Texts=Make Array!(
     "[0]":Text=Format Text!(
      Format="LOCTEXT(\"Here are the times when some of the Ranked lists are enabled ({TimeZoneIdentifier}) for Battle Royale Build:\")"
      TimeZoneIdentifier:Text=@n11
     ) -> (Result)
     "[1]":Text=Format Text!(Format="LOCTEXT(\"{1}\\r\\n{2}\\r\\n{3}\\r\\n{4}\\r\\n{5}\")", 1:Text=@n9, 2:Text=@n10, 3:Text=@n6, 4:Text=@n7, 5:Text=@n8) -> (Result)
     "[2]":Text=Format Text!(Format="LOCTEXT(\"For Battle Royale Zero Build:\")") -> (Result)
     "[3]":Text=Format Text!(Format="LOCTEXT(\"{1}\\r\\n{2}\\r\\n{3}\\r\\n{4}\\r\\n{5}\")", 1:Text=@n2, 2:Text=@n3, 3:Text=@n4, 4:Text=@n5, 5:Text=@n1) -> (Result)
    ) -> (Array:Array of Texts)
   )
  ]}
]
variables [DebugHoursOffsetForTestingTimeZones {type=integer}]
widgets {
 info {
  inheritedWidgetCount=0
  parentClass=&/WBP_Habanero_GenericEntry.WBP_Habanero_GenericEntry_C
  rootWidgetClass=null
  widgetCount=0
 }
 widgets []
}
