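# Lobby info schedule widget, derived from WBP_Habanero_TimeAwareScheduleEntry, listing when regional Ranked playlists are enabled for Reload Build and Zero Build.
# UpdateBullets casts to BP_UI_Habanero_InfoPanel, sets its Details List from MakeDynamicTimeAwareText, then calls the parent.
# That function formats data-table rows per region into four localized text blocks, labelled with a time zone identifier.
# Zero Build's Oceania Squads hours read the non-ZB blastberry-oceania row.
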WidgetBlueprint /Game/UI/Frontend/Lobby/Habanero/HabaneroInfoEntry/WBP_Habanero_TimeAwareScheduleEntry_Blastberry
graphs [
 EventGraph {nodes [
   n1 = Event UpdateBullets!() -> (OutputDelegate, then, Object:Object Reference) {is=title K2Node_Event}
   > n2 = Cast To BP_UI_Habanero_InfoPanel!(execute, Object=@n1.Object) -> (then, CastFailed, AsBP UI Habanero Info Panel:BP UI Habanero Info Panel Object Reference) {is=title K2Node_DynamicCast}
   > Set Details List!(
    execute
    Details List:Array of Texts=MakeDynamicTimeAwareText!() -> (Array:Array of Texts)
    self:BP UI Habanero Info Panel Object Reference=@n2
   ) -> (then, Output_Get:Array of Texts)
   > Parent: UpdateBullets!(execute, Object:Object Reference=@n2) -> (then) {is=title K2Node_CallParentFunction}
  ]}
 MakeDynamicTimeAwareText {nodes [
   MakeDynamicTimeAwareText() {is=title K2Node_FunctionEntry}
   n1 = FormatBulletPointTextForRegionFromTable!(RegionName:Text="LOCTEXT(\"Brazil\")", Type:Text="LOCTEXT(\"Squads\")", RegionDataTableRowName:Name=blastberry-brazil) -> (Result:Text)
   n2 = FormatBulletPointTextForRegionFromTable(RegionName="LOCTEXT(\"Asia\")", Type="LOCTEXT(\"Solo\")", RegionDataTableRowName=blastberry-asia)
   n3 = FormatBulletPointTextForRegionFromTable(RegionName="LOCTEXT(\"NA-East, NA-Central, NA-West, Europe and Middle East\")", RegionDataTableRowName=blastberry-na-east-and-europe)
   n4 = FormatBulletPointTextForRegionFromTable(RegionName="LOCTEXT(\"Oceania\")", Type="LOCTEXT(\"Solo\")", RegionDataTableRowName=blastberry-oceania)
   n5 = FormatTimeZoneIdentifierText!() -> (ReturnValue:Text)
   n6 = FormatBulletPointTextForRegionFromTable(RegionName="LOCTEXT(\"Middle East\")", Type="LOCTEXT(\"Squads\")", RegionDataTableRowName=blastberryzb-middle-east)
   n7 = FormatBulletPointTextForRegionFromTable(RegionName="LOCTEXT(\"NA-East, NA-Central, NA-West and Europe\")", RegionDataTableRowName=blastberryzb-na-east-and-europe)
   n8 = FormatBulletPointTextForRegionFromTable(RegionName="LOCTEXT(\"Brazil\")", Type="LOCTEXT(\"Squads\")", RegionDataTableRowName=blastberryzb-brazil)
   n9 = FormatBulletPointTextForRegionFromTable(RegionName="LOCTEXT(\"Asia\")", Type="LOCTEXT(\"Squads\")", RegionDataTableRowName=blastberryzb-asia)
   n10 = Format Text!(
    Format="LOCTEXT(\"{1}, {2}\")"
    1:Text=FormatBulletPointTextForRegionFromTable(RegionName="LOCTEXT(\"Oceania\")", Type="LOCTEXT(\"Solos/Duo\")", RegionDataTableRowName=blastberryzb-oceania)
    2:Text=Format Text!(
     Format="LOCTEXT(\"Squads - {RegionAvailableTime}\")"
     RegionAvailableTime:Text=FormatHoursFromDataTable!(Region:Name=blastberry-oceania) -> (ReturnValue:Text)
    ) -> (Result)
   ) -> (Result)
   > Return Node!(
    execute
    Array:Array of Texts=Make Array!(
     "[0]":Text=Format Text!(
      Format="LOCTEXT(\"Here are the times when some of the Ranked lists are enabled ({TimeZoneIdentifier}) for Reload Build:\")"
      TimeZoneIdentifier:Text=@n5
     ) -> (Result)
     "[1]":Text=Format Text!(Format="LOCTEXT(\"{1}\\r\\n{2}\\r\\n{3}\\r\\n{4}\")", 1:Text=@n3, 2:Text=@n4, 3:Text=@n1, 4:Text=@n2) -> (Result)
     "[2]":Text=Format Text!(Format="LOCTEXT(\"For Reload Zero Build:\")") -> (Result)
     "[3]":Text=Format Text!(Format="LOCTEXT(\"{1}\\r\\n{2}\\r\\n{3}\\r\\n{4}\\r\\n{5}\")", 1:Text=@n7, 2:Text=@n10, 3:Text=@n8, 4:Text=@n9, 5:Text=@n6) -> (Result)
    ) -> (Array:Array of Texts)
   )
  ]}
]
variables []
widgets {
 info {
  inheritedWidgetCount=0
  parentClass=&/WBP_Habanero_TimeAwareScheduleEntry.WBP_Habanero_TimeAwareScheduleEntry_C
  rootWidgetClass=null
  widgetCount=0
 }
 widgets []
}
